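# Control Rig for the F_MED_BentBaton_Flip cosmetic body, procedurally posing its skirt base bones from leg motion.
# After a shared deform-rig function, spherical pose readers on both thighs store forward, side, back, side-in and twist-in alphas.
# Sequenced stages constrain front and side skirt bones to the thighs, then layer corrective offsets and blended constraints, often using interpolated left/right differences.
# Frontend idle and emote variables adjust some corrections; thighSide alphas are never read.
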ControlRigBlueprint /BRCosmetics/Characters/Player/Female/Medium/Bodies/F_MED_BentBaton_Flip/Meshes/F_MED_BentBaton_Flip_CtrlRig
graphs [
 RigVMFunctionLibrary {links [], nodes []}
 RigVMFunction_Sequence_1_ContainedGraph {nodes [
   n1 = RigVMFunctionEntryNode()
   RigVMFunction_Sequence = &RigVMFunction_Sequence(ExecuteContext=@n1.ExecuteContext)
   n2 = &RigVMFunction_Sequence(ExecuteContext=@RigVMFunction_Sequence.B)
   n3 = &RigVMFunction_Sequence(ExecuteContext=@n2.B)
   n4 = &RigVMFunction_Sequence(ExecuteContext=@n3.B)
   n5 = &RigVMFunction_Sequence(ExecuteContext=@n4.B)
   n6 = &RigVMFunction_Sequence(ExecuteContext=@n5.B)
   RigVMFunctionReturnNode(A=@RigVMFunction_Sequence.A, B=@n2.A, C=@n3.A, D=@n4.A, E=@n5.A, F=@n6.A, G=@n6.B)
  ]}
 RigVMModel {nodes [
   RigUnit_BeginExecution = &RigUnit_BeginExecution()
   RigVMFunction_Sequence = RigVMAggregateNode(ExecuteContext=@RigUnit_BeginExecution.ExecutePin)
   n1 = &RigUnit_SphericalPoseReader(
    ActiveRegionScaleFactors=(PositiveWidth=4,NegativeWidth=3,PositiveHeight=2.5,NegativeHeight=2)
    ActiveRegionSize=0.1
    Debug=(bDrawDebug=false,bDraw2D=False,bDrawLocalAxes=False,DebugScale=25,DebugSegments=20,DebugThickness=0.25)
    DriverAxis=(X=-1,Y=0,Z=0)
    DriverItem=(Type=Bone,Name="thigh_l")
    FalloffRegionScaleFactors=(PositiveWidth=1,NegativeWidth=4,PositiveHeight=1,NegativeHeight=0.5)
    FalloffSize=0.4
    FlipHeightScaling=False
    FlipWidthScaling=False
    OptionalParentItem=(Type=Bone,Name="None")
    RotationOffset=(X=100,Y=0,Z=-20)
    ExecutePin=@RigVMFunction_Sequence.B
   )
   n2 = RigVMVariableNode(Value=@n1.OutputParam, ExecuteContext=@RigVMFunction_Sequence.C) {variable=L thighForward Alpha}
   n3 = &RigUnit_SphericalPoseReader(
    ActiveRegionScaleFactors=(PositiveWidth=4,NegativeWidth=3,PositiveHeight=2.5,NegativeHeight=2)
    ActiveRegionSize=0.1
    Debug=(bDrawDebug=true,bDraw2D=False,bDrawLocalAxes=False,DebugScale=25,DebugSegments=20,DebugThickness=0.25)
    DriverAxis=(X=1,Y=0,Z=0)
    DriverItem=(Type=Bone,Name="thigh_r")
    FalloffRegionScaleFactors=(PositiveWidth=1,NegativeWidth=4,PositiveHeight=1,NegativeHeight=0.5)
    FalloffSize=0.4
    FlipHeightScaling=False
    FlipWidthScaling=true
    OptionalParentItem=(Type=Bone,Name="None")
    RotationOffset=(X=-80,Y=0,Z=-20)
    ExecutePin=@n1.ExecutePin
   )
   n4 = RigVMVariableNode(Value=@n3.OutputParam, ExecuteContext=@n2.ExecuteContext) {variable=R thighForward Alpha}
   n5 = &RigUnit_SphericalPoseReader(
    ActiveRegionScaleFactors=(PositiveWidth=3,NegativeWidth=2,PositiveHeight=2,NegativeHeight=2)
    ActiveRegionSize=0.1
    Debug=(bDrawDebug=false,bDraw2D=False,bDrawLocalAxes=False,DebugScale=25,DebugSegments=20,DebugThickness=0.25)
    DriverAxis=(X=-1,Y=0,Z=0)
    DriverItem=(Type=Bone,Name="thigh_l")
    FalloffRegionScaleFactors=(PositiveWidth=2,NegativeWidth=2,PositiveHeight=2,NegativeHeight=2)
    FalloffSize=0.25
    FlipHeightScaling=false
    FlipWidthScaling=false
    OptionalParentItem=(Type=Bone,Name="None")
    RotationOffset=(X=180,Y=-40,Z=0)
    ExecutePin=@n3.ExecutePin
   )
   n6 = RigVMVariableNode(Value=@n5.OutputParam, ExecuteContext=@n4.ExecuteContext) {variable=L thighSide Alpha}
   n7 = &RigUnit_SphericalPoseReader(
    ActiveRegionScaleFactors=(PositiveWidth=3,NegativeWidth=2,PositiveHeight=2,NegativeHeight=2)
    ActiveRegionSize=0.1
    Debug=(bDrawDebug=false,bDraw2D=False,bDrawLocalAxes=False,DebugScale=25,DebugSegments=20,DebugThickness=0.25)
    DriverAxis=(X=1,Y=0,Z=0)
    DriverItem=(Type=Bone,Name="thigh_r")
    FalloffRegionScaleFactors=(PositiveWidth=2,NegativeWidth=2,PositiveHeight=2,NegativeHeight=2)
    FalloffSize=0.25
    FlipHeightScaling=False
    FlipWidthScaling=true
    OptionalParentItem=(Type=Bone,Name="None")
    RotationOffset=(X=0,Y=-40,Z=0)
    ExecutePin=@n5.ExecutePin
   )
   n8 = RigVMVariableNode(Value=@n7.OutputParam, ExecuteContext=@n6.ExecuteContext) {variable=R thighSide Alpha}
   n9 = &RigUnit_SphericalPoseReader(
    ActiveRegionScaleFactors=(PositiveWidth=2,NegativeWidth=2,PositiveHeight=2,NegativeHeight=1)
    ActiveRegionSize=0.1
    Debug=(bDrawDebug=false,bDraw2D=False,bDrawLocalAxes=False,DebugScale=25,DebugSegments=20,DebugThickness=0.25)
    DriverAxis=(X=-1,Y=0,Z=0)
    DriverItem=(Type=Bone,Name="thigh_l")
    FalloffRegionScaleFactors=(PositiveWidth=2,NegativeWidth=2,PositiveHeight=4,NegativeHeight=1)
    FalloffSize=0.3
    FlipHeightScaling=false
    FlipWidthScaling=false
    OptionalParentItem=(Type=Bone,Name="None")
    RotationOffset=(X=0,Y=90,Z=80)
    ExecutePin=@n7.ExecutePin
   )
   n10 = RigVMVariableNode(Value=@n9.OutputParam, ExecuteContext=@n8.ExecuteContext) {variable=L thighBack Alpha}
   n11 = &RigUnit_SphericalPoseReader(
    ActiveRegionScaleFactors=(PositiveWidth=2,NegativeWidth=2,PositiveHeight=2,NegativeHeight=1)
    ActiveRegionSize=0.1
    Debug=(bDrawDebug=false,bDraw2D=False,bDrawLocalAxes=False,DebugScale=25,DebugSegments=20,DebugThickness=0.25)
    DriverAxis=(X=1,Y=0,Z=0)
    DriverItem=(Type=Bone,Name="thigh_r")
    FalloffRegionScaleFactors=(PositiveWidth=2,NegativeWidth=2,PositiveHeight=3,NegativeHeight=1)
    FalloffSize=0.3
    FlipHeightScaling=false
    FlipWidthScaling=false
    OptionalParentItem=(Type=Bone,Name="None")
    RotationOffset=(X=180,Y=93.7803,Z=60.6503)
    ExecutePin=@n9.ExecutePin
   )
   n12 = RigVMVariableNode(Value=@n11.OutputParam, ExecuteContext=@n10.ExecuteContext) {variable=R thighBack Alpha}
   n13 = &RigUnit_SphericalPoseReader(
    ActiveRegionScaleFactors=(PositiveWidth=3,NegativeWidth=2,PositiveHeight=2,NegativeHeight=2)
    ActiveRegionSize=0.1
    Debug=(bDrawDebug=false,bDraw2D=False,bDrawLocalAxes=False,DebugScale=25,DebugSegments=20,DebugThickness=0.25)
    DriverAxis=(X=-1,Y=0,Z=0)
    DriverItem=(Type=Bone,Name="thigh_l")
    FalloffRegionScaleFactors=(PositiveWidth=2,NegativeWidth=2,PositiveHeight=2,NegativeHeight=2)
    FalloffSize=0.25
    FlipHeightScaling=false
    FlipWidthScaling=false
    OptionalParentItem=(Type=Bone,Name="None")
    RotationOffset=(X=2.688,Y=38.2586,Z=-2.2671)
    ExecutePin=@n11.ExecutePin
   )
   n14 = RigVMVariableNode(Value=@n13.OutputParam, ExecuteContext=@n12.ExecuteContext) {variable=L thighSideIn Alpha}
   n15 = &RigUnit_SphericalPoseReader(
    ActiveRegionScaleFactors=(PositiveWidth=3,NegativeWidth=2,PositiveHeight=2,NegativeHeight=2)
    ActiveRegionSize=0.1
    Debug=(bDrawDebug=false,bDraw2D=False,bDrawLocalAxes=False,DebugScale=25,DebugSegments=20,DebugThickness=0.25)
    DriverAxis=(X=1,Y=0,Z=0)
    DriverItem=(Type=Bone,Name="thigh_r")
    FalloffRegionScaleFactors=(PositiveWidth=2,NegativeWidth=2,PositiveHeight=2,NegativeHeight=2)
    FalloffSize=0.25
    FlipHeightScaling=False
    FlipWidthScaling=true
    OptionalParentItem=(Type=Bone,Name="None")
    RotationOffset=(X=0,Y=-40,Z=0)
    ExecutePin=@n13.ExecutePin
   )
   n16 = RigVMVariableNode(Value=@n15.OutputParam, ExecuteContext=@n14.ExecuteContext) {variable=R thighSideIn Alpha}
   n17 = &RigUnit_SphericalPoseReader(
    ActiveRegionScaleFactors=(PositiveWidth=1,NegativeWidth=1,PositiveHeight=1,NegativeHeight=1)
    ActiveRegionSize=0.12
    Debug=(bDrawDebug=true,bDraw2D=False,bDrawLocalAxes=False,DebugScale=25,DebugSegments=20,DebugThickness=0.25)
    DriverAxis=(X=0,Y=1,Z=0)
    DriverItem=(Type=Bone,Name="thigh_l")
    FalloffRegionScaleFactors=(PositiveWidth=1,NegativeWidth=1,PositiveHeight=1,NegativeHeight=1)
    FalloffSize=0.17
    FlipHeightScaling=false
    FlipWidthScaling=false
    OptionalParentItem=(Type=Bone,Name="None")
    RotationOffset=(X=58.5993,Y=-9.6785,Z=-26.2186)
    ExecutePin=@n15.ExecutePin
   )
   n18 = RigVMVariableNode(Value=@n17.OutputParam, ExecuteContext=@n16.ExecuteContext) {variable=L thighTwistIn Alpha}
   n19 = &RigUnit_SphericalPoseReader(
    ActiveRegionScaleFactors=(PositiveWidth=1,NegativeWidth=1,PositiveHeight=1,NegativeHeight=1)
    ActiveRegionSize=0.12
    Debug=(bDrawDebug=true,bDraw2D=False,bDrawLocalAxes=False,DebugScale=25,DebugSegments=20,DebugThickness=0.25)
    DriverAxis=(X=0,Y=-1,Z=0)
    DriverItem=(Type=Bone,Name="thigh_r")
    FalloffRegionScaleFactors=(PositiveWidth=1,NegativeWidth=1,PositiveHeight=1,NegativeHeight=1)
    FalloffSize=0.17
    FlipHeightScaling=False
    FlipWidthScaling=true
    OptionalParentItem=(Type=Bone,Name="None")
    RotationOffset=(X=-127.0157,Y=4.3906,Z=-18.4663)
    ExecutePin=@n17.ExecutePin
   )
   RigVMVariableNode(Value=@n19.OutputParam, ExecuteContext=@n18.ExecuteContext) {variable=R thighTwistIn Alpha}
   n20 = &RigVMFunction_MathDoubleSub(B=1, A=RigVMVariableNode() {variable=IsPlayingFrontendIdle}.Value)
   &RigUnit_ModifyTransforms(
    ItemToModify="((Item=(Type=Bone,Name=\"dyn_skirt_fr_l_base\"),Transform=(Rotation=(X=0,Y=-0.2578,Z=0.0872,W=0.9623),Translation=(X=0,Y=0,Z=0),Scale3D=(X=1,Y=1,Z=1))),(Item=(Type=Bone,Name=\"dyn_skirt_fr_c_base\"),Transform=(Rotation=(X=0,Y=0.1736,Z=0,W=-0.9848),Translation= ...(truncated)"
    WeightMinimum=0
    Weight=&RigVMFunction_MathDoubleMul(B=&RigVMFunction_MathDoubleAbs(Value=@n20.Result).Result, A=RigVMVariableNode() {variable=L thighSideIn Alpha}.Value).Result
    ExecutePin=@RigVMFunction_Sequence.G
   )
   > &RigUnit_ModifyTransforms(
    ItemToModify="((Item=(Type=Bone,Name=\"dyn_skirt_fr_l_base\"),Transform=(Rotation=(X=0,Y=0,Z=0.0872,W=0.9962),Translation=(X=0,Y=0,Z=0),Scale3D=(X=1,Y=1,Z=1))),(Item=(Type=Bone,Name=\"dyn_skirt_fr_c_base\"),Transform=(Rotation=(X=0,Y=0,Z=0,W=1),Translation=(X=0,Y=0,Z=0),Sca ...(truncated)"
    WeightMinimum=0
    Weight=RigVMVariableNode() {variable=IsPlayingFrontendIdle}.Value
   )
   n21 = RigVMVariableNode() {variable=L thighTwistIn Alpha}
   n22 = RigVMVariableNode() {variable=R thighTwistIn Alpha}
   n23 = &RigVMFunction_AlphaInterp(
    InterpSpeedDecreasing=10
    InterpSpeedIncreasing=10
    Scale=1
    bClampResult=true
    Value=&RigVMFunction_MathDoubleMul(A=@n21.Value, B=@n22.Value).Result
   )
   > &RigUnit_ModifyTransforms(
    ItemToModify="((Item=(Type=Bone,Name=\"dyn_skirt_fr_l_base\"),Transform=(Rotation=(X=0.0151,Y=-0.173,Z=-0.0858,W=0.9811),Translation=(X=3.0659,Y=-0.5753,Z=1.1159),Scale3D=(X=1,Y=1,Z=1))),(Item=(Type=Bone,Name=\"dyn_skirt_fr_c_base\"),Transform=(Rotation=(X=0,Y=-0.2588,Z=0,W ...(truncated)"
    WeightMinimum=0
    Weight=@n23.Result
   )
   n24 = &RigVMFunction_AlphaInterp(
    InterpSpeedDecreasing=10
    InterpSpeedIncreasing=10
    Scale=1
    bClampResult=true
    Value=&RigVMFunction_MathDoubleSub(A=@n21.Value, B=@n22.Value).Result
   )
   > &RigUnit_ModifyTransforms(
    ItemToModify="((Item=(Type=Bone,Name=\"dyn_skirt_fr_c_base\"),Transform=(Rotation=(X=0,Y=0,Z=0,W=-1),Translation=(X=0,Y=0,Z=0),Scale3D=(X=1,Y=1,Z=1))),(Item=(Type=Bone,Name=\"dyn_skirt_fr_l_base\"),Transform=(Rotation=(X=0,Y=0,Z=-0.2588,W=0.9659),Translation=(X=0,Y=0,Z=0),S ...(truncated)"
    WeightMinimum=0
    Weight=@n24.Result
   )
   n25 = &RigVMFunction_AlphaInterp(
    InterpSpeedDecreasing=10
    InterpSpeedIncreasing=10
    Scale=1
    bClampResult=true
    Value=&RigVMFunction_MathDoubleSub(B=@n21.Value, A=@n22.Value).Result
   )
   > &RigUnit_ModifyTransforms(
    ItemToModify="((Item=(Type=Bone,Name=\"dyn_skirt_fr_c_base\"),Transform=(Rotation=(X=0,Y=0,Z=0,W=1),Translation=(X=0,Y=0,Z=0),Scale3D=(X=1,Y=1,Z=1))),(Item=(Type=Bone,Name=\"dyn_skirt_fr_r_base\"),Transform=(Rotation=(X=0,Y=0,Z=0,W=1),Translation=(X=0,Y=0,Z=0),Scale3D=(X=1, ...(truncated)"
    WeightMinimum=0
    Weight=@n25.Result
   )
   FN_DeformRig_FNC = RigVMFunctionReferenceNode(
    DeformRig_ Curve="(deformRig_l_upperarm_up_alpha_72_#1=False,deformRig_r_upperarm_up_alpha_73_#2=False,deformRig_l_upperarm_fr_alpha_74_#3=False,deformRig_r_upperarm_fr_alpha_75_CB51B8 ...(truncated)"
    DeformRig_Bool="(upperarm_1_#4=True,elbow_3_#5=false,wrist_7_#6=false,groin_8_#7=false,glute_17_#8=false,knee_9_40C408324C ...(truncated)"
    DeformRig_Debug=(upperarm_debug_18_#4=False,wrist_debug_20_#6=False,lat_debug_26_#9=False,groinGlute_debug_27_#7=False)
    DeformRig_Value="(upperarm_up_18_#10=(Rotation=(X=0,Y=0,Z=0,W=1),Translation=(X=0,Y=0,Z=0),Scale3D=(X=0.7,Y=1,Z=1)),upperarm_fwd_fr_19_#11=(Rotation=(X=0,Y=0,Z=0,W=1),Translation=(X=0,Y=0,Z=0),Scale3D=(X=1,Y=1,Z=1)) ...(truncated)"
    ExecuteContext=@RigVMFunction_Sequence.A
   )
   n26 = RigVMVariableNode() {variable=L thighForward Alpha}
   n27 = RigVMVariableNode() {variable=R thighForward Alpha}
   Subtract = &RigVMFunction_MathDoubleSub(A=@n26.Value, B=@n27.Value)
   n28 = &RigVMFunction_AlphaInterp(
    InterpSpeedDecreasing=30
    InterpSpeedIncreasing=30
    Scale=1
    bClampResult=true
    Value=&RigVMFunction_MathDoubleMul(B=0.5, A=@Subtract.Result).Result
   )
   &RigUnit_ParentConstraint(
    AdvancedSettings=(InterpolationType=Average,RotationOrderForFilter=XZY)
    Child=(Type=Bone,Name="dyn_skirt_fr_r_base")
    Parents=((Item=(Type=Bone,Name="dyn_skirt_fr_c_base"),Weight=0.1))
    Weight=@n28.Result
    ExecutePin=@RigVMFunction_Sequence.F
   )
   n29 = &RigVMFunction_MathDoubleMul(B=0.35, A=@Subtract.Result)
   n30 = &RigVMFunction_AlphaInterp(InterpSpeedDecreasing=30, InterpSpeedIncreasing=30, Scale=1, bClampResult=true, Value=@n29.Result)
   > &RigUnit_ParentConstraint(
    AdvancedSettings=(InterpolationType=Average,RotationOrderForFilter=XZY)
    Child=(Type=Bone,Name="dyn_skirt_side_r_base")
    Parents=((Item=(Type=Bone,Name="dyn_skirt_fr_r_base"),Weight=0.1))
    Weight=@n30.Result
   )
   n31 = &RigVMFunction_MathDoubleSub(B=@n26.Value, A=@n27.Value)
   Multiply = &RigVMFunction_MathDoubleMul(B=0.5, A=@n31.Result)
   n32 = &RigVMFunction_AlphaInterp(
    InterpSpeedDecreasing=30
    InterpSpeedIncreasing=30
    Scale=1
    bClampResult=true
    Value=&RigVMFunction_MathDoubleSub(B=0, A=@Multiply.Result).Result
   )
   > &RigUnit_ParentConstraint(
    AdvancedSettings=(InterpolationType=Average,RotationOrderForFilter=XZY)
    Child=(Type=Bone,Name="dyn_skirt_fr_l_base")
    Parents=((Item=(Type=Bone,Name="dyn_skirt_fr_c_base"),Weight=0.1))
    Weight=@n32.Result
   )
   > &RigUnit_ParentConstraint(
    AdvancedSettings=(InterpolationType=Average,RotationOrderForFilter=XZY)
    Child=(Type=Bone,Name="dyn_skirt_side_l_base")
    Parents=((Item=(Type=Bone,Name="dyn_skirt_fr_l_base"),Weight=0.1))
    Weight=@Subtract.Result
   )
   Clamp = &RigVMFunction_MathDoubleClamp(Value=@Subtract.Result)
   n33 = &RigVMFunction_MathDoubleClamp(Value=@n31.Result)
   Add = &RigVMFunction_MathDoubleAdd(A=@Clamp.Result, B=@n33.Result)
   n34 = &RigVMFunction_AlphaInterp(InRange=(Min=1,Max=0), InterpSpeedDecreasing=30, InterpSpeedIncreasing=30, Scale=1, bClampResult=true, bMapRange=true, Value=@Add.Result)
   > &RigUnit_ParentConstraint(
    AdvancedSettings=(InterpolationType=Average,RotationOrderForFilter=XZY)
    Child=(Type=Bone,Name="dyn_skirt_fr_c_base")
    Parents=((Item=(Type=Bone,Name="dyn_skirt_fr_l_base"),Weight=0.5),(Item=(Type=Bone,Name="dyn_skirt_fr_r_base"),Weight=0.5))
    Weight=@n34.Result
   )
   n35 = &RigVMFunction_AlphaInterp(InterpSpeedDecreasing=30, InterpSpeedIncreasing=30, Scale=1, bClampResult=true, Value=@Clamp.Result)
   > &RigUnit_ModifyTransforms(
    ItemToModify="((Item=(Type=Bone,Name=\"dyn_skirt_fr_c_base\"),Transform=(Rotation=(X=0.3725,Y=0.3815,Z=0.081,W=-0.8421),Translation=(X=0,Y=0,Z=0),Scale3D=(X=1,Y=1,Z=1))),(Item=(Type=Bone,Name=\"dyn_skirt_fr_r_base\"),Transform=(Rotation=(X=-0.13,Y=-0.2566,Z=0.0114,W=0.9577) ...(truncated)"
    WeightMinimum=0
    Weight=@n35.Result
   )
   Absolute = &RigVMFunction_MathDoubleAbs(Value=&RigVMFunction_MathDoubleSub(B=1, A=RigVMVariableNode() {variable=IsPlayingFrontendIdle}.Value).Result)
   n36 = RigVMVariableNode() {variable=IsInFrontendEmote}
   n37 = &RigVMFunction_AlphaInterp(
    InterpSpeedDecreasing=30
    InterpSpeedIncreasing=30
    Scale=1
    bClampResult=true
    Value=&RigVMFunction_MathDoubleMul(A=&RigVMFunction_MathDoubleAdd(A=@Absolute.Result, B=@n36.Value).Result, B=@n33.Result).Result
   )
   > &RigUnit_ModifyTransforms(
    ItemToModify="((Item=(Type=Bone,Name=\"dyn_skirt_fr_c_base\"),Transform=(Rotation=(X=0.2778,Y=-0.4404,Z=0.1151,W=0.846),Translation=(X=0,Y=0,Z=0),Scale3D=(X=1,Y=1,Z=1))),(Item=(Type=Bone,Name=\"dyn_skirt_fr_l_base\"),Transform=(Rotation=(X=0.2537,Y=-0.2554,Z=0.0696,W=0.9304 ...(truncated)"
    WeightMinimum=0
    Weight=@n37.Result
   )
   > &RigUnit_ParentConstraint(
    AdvancedSettings=(InterpolationType=Average,RotationOrderForFilter=XZY)
    Child=(Type=Bone,Name="dyn_skirt_bck_r_base")
    Parents=((Item=(Type=Bone,Name="dyn_skirt_side_r_base"),Weight=1))
    Weight=0.1
   )
   > &RigUnit_ParentConstraint(
    AdvancedSettings=(InterpolationType=Average,RotationOrderForFilter=XZY)
    Child=(Type=Bone,Name="dyn_skirt_bck_l_base")
    Parents=((Item=(Type=Bone,Name="dyn_skirt_side_l_base"),Weight=1))
    Weight=0.1
   )
   &RigUnit_ParentConstraint(
    AdvancedSettings=(InterpolationType=Average,RotationOrderForFilter=XZY)
    Child=(Type=Bone,Name="dyn_skirt_side_l_base")
    Parents=((Item=(Type=Bone,Name="thigh_l"),Weight=1))
    Weight=1
    ExecutePin=@RigVMFunction_Sequence.D
   )
   > &RigUnit_ParentConstraint(
    AdvancedSettings=(InterpolationType=Average,RotationOrderForFilter=XZY)
    Child=(Type=Bone,Name="dyn_skirt_fr_l_base")
    Parents=((Item=(Type=Bone,Name="thigh_l"),Weight=1),(Item=(Type=Bone,Name="thigh_r"),Weight=0))
    Weight=1
   )
   > &RigUnit_ParentConstraint(
    AdvancedSettings=(InterpolationType=Average,RotationOrderForFilter=XZY)
    Child=(Type=Bone,Name="dyn_skirt_fr_r_base")
    Parents=((Item=(Type=Bone,Name="thigh_r"),Weight=1),(Item=(Type=Bone,Name="thigh_l"),Weight=0))
    Weight=1
   )
   > &RigUnit_ParentConstraint(
    AdvancedSettings=(InterpolationType=Average,RotationOrderForFilter=XZY)
    Child=(Type=Bone,Name="dyn_skirt_side_r_base")
    Parents=((Item=(Type=Bone,Name="thigh_r"),Weight=1))
    Weight=1
   )
   > &RigUnit_ParentConstraint(
    AdvancedSettings=(InterpolationType=Average,RotationOrderForFilter=XZY)
    Child=(Type=Bone,Name="dyn_skirt_fr_c_base")
    Parents=((Item=(Type=Bone,Name="dyn_skirt_fr_l_base"),Weight=0.5),(Item=(Type=Bone,Name="dyn_skirt_fr_r_base"),Weight=0.5))
    Weight=1
   )
   &RigUnit_ModifyTransforms(
    ItemToModify=((Item=(Type=Bone,Name="dyn_skirt_fr_l_base"),Transform=(Rotation=(X=-0.1048,Y=-0.0878,Z=0.1671,W=0.9764),Translation=(X=0,Y=0,Z=0),Scale3D=(X=1,Y=1,Z=1))))
    WeightMinimum=0
    Weight=RigVMVariableNode() {variable=L thighForward Alpha}.Value
    ExecutePin=@RigVMFunction_Sequence.E
   )
   > &RigUnit_ModifyTransforms(
    ItemToModify=((Item=(Type=Bone,Name="dyn_skirt_fr_r_base"),Transform=(Rotation=(X=0,Y=0.1736,Z=0,W=-0.9848),Translation=(X=0,Y=0,Z=0),Scale3D=(X=1,Y=1,Z=1))))
    WeightMinimum=0
    Weight=RigVMVariableNode() {variable=R thighForward Alpha}.Value
   )
   > &RigUnit_ModifyTransforms(
    ItemToModify="((Item=(Type=Bone,Name=\"dyn_skirt_bck_l_base\"),Transform=(Rotation=(X=-0.0019,Y=-0.2162,Z=-0.0436,W=0.9754),Translation=(X=0,Y=0,Z=0),Scale3D=(X=1,Y=1,Z=1))),(Item=(Type=Bone,Name=\"dyn_skirt_bck_c_base\"),Transform=(Rotation=(X=0,Y=-0.0872,Z=0,W=0.9962),Tra ...(truncated)"
    WeightMinimum=0
    Weight=RigVMVariableNode() {variable=L thighBack Alpha}.Value
   )
   > &RigUnit_ModifyTransforms(
    ItemToModify="((Item=(Type=Bone,Name=\"dyn_skirt_bck_r_base\"),Transform=(Rotation=(X=-0.0038,Y=-0.0871,Z=0.0435,W=0.9952),Translation=(X=0,Y=0,Z=0),Scale3D=(X=1,Y=1,Z=1))),(Item=(Type=Bone,Name=\"dyn_skirt_bck_c_base\"),Transform=(Rotation=(X=0,Y=-0.0872,Z=0,W=0.9962),Tran ...(truncated)"
    WeightMinimum=0
    Weight=RigVMVariableNode() {variable=R thighBack Alpha}.Value
   )
   RerouteNode_14_Constant = RigVMDispatchNode(Value=0) {template="DISPATCH_RigVMDispatch_Constant(io Value)"}
   &RigVMFunction_AlphaInterp(
    InterpSpeedDecreasing=30
    InterpSpeedIncreasing=30
    Scale=1
    bClampResult=False
    Value=&RigVMFunction_MathDoubleSub(A=&RigVMFunction_MathDoubleMul(B=0.35, A=@n31.Result).Result, B=@RerouteNode_14_Constant.Value).Result
   )
   &RigVMFunction_MathDoubleSub(A=0, B=0)
   &RigVMFunction_MathDoubleSub(B=0, A=@n29.Result)
  ]}
]
hierarchy [
 root {loc=<0 0 0>, rot=<0 0 0 1>}
 pelvis {parent=root, loc=<0 -0.5301 98.0571>, rot=<0.0127 0.707 -0.0127 -0.707>}
 spine_01 {parent=pelvis, loc=<6.1208 -0.9328 0>, rot=<0 0 0.034 -0.9994>}
 spine_02 {parent=spine_01, loc=<6.5997 -0.2095 0>, rot=<0 0 0.0238 -0.9997>}
 spine_03 {parent=spine_02, loc=<6.4914 -0.1237 0>, rot=<0 0 0.0149 -0.9999>}
 spine_04 {parent=spine_03, loc=<6.9598 0.0794 0>, rot=<0 0 -0.0132 -0.9999>}
 spine_05 {parent=spine_04, loc=<6.9987 0 0>, rot=<0 0 -0.0015 -1>}
 neck_01 {parent=spine_05, loc=<11.3138 0.1529 0>, rot=<0 0 -0.2003 -0.9797>}
 neck_02 {parent=neck_01, loc=<4.5627 0 0>, rot=<0 0 0.0504 -0.9987>}
 head {parent=neck_02, loc=<5.3914 0 0>, rot=<0 0 0.115 -0.9934>}
 clavicle_l {parent=spine_05, loc=<5.4831 -0.1053 -3.2946>, rot=<0.0465 -0.7289 -0.009 -0.6829>}
 upperarm_l {parent=clavicle_l, loc=<9.8658 -1.2136 -0.4105>, rot=<0.0164 -0.3245 -0.0269 -0.9454>}
 lowerarm_l {parent=upperarm_l, loc=<26.465 -0.8345 0.259>, rot=<0.0036 0.0161 -0.2298 -0.9731>}
 hand_l {parent=lowerarm_l, loc=<22.7355 -0.3203 -0.6105>, rot=<-0.6784 -0.0223 0.0735 -0.7307>}
 thumb_01_l {parent=hand_l, loc=<1.3118 -0.3191 -2.459>, rot=<0.6546 -0.4042 0.0664 -0.6354>}
 thumb_02_l {parent=thumb_01_l, loc=<5.9336 -0.0448 0.0401>, rot=<-0.0037 0.0213 0.298 -0.9543>}
 thumb_03_l {parent=thumb_02_l, loc=<2.2223 0.3682 -0.0907>, rot=<0.0118 -0.013 0.0066 -0.9998>}
 index_metacarpal_l {parent=hand_l, loc=<1.6535 0.0183 -2.96>, rot=<0.0158 -0.0124 -0.0173 -0.9996>}
 index_01_l {parent=index_metacarpal_l, loc=<8.3008 0.0357 -0.0524>, rot=<-0.0044 0.0485 0.0924 -0.9945>}
 index_02_l {parent=index_01_l, loc=<3.3587 -0.3004 0.0452>, rot=<-0.002 0.0027 0.3225 -0.9466>}
 index_03_l {parent=index_02_l, loc=<2.5578 -0.123 -0.0252>, rot=<-0.002 0.0067 0.2558 -0.9667>}
 middle_metacarpal_l {parent=hand_l, loc=<2.1856 -0.1725 -0.5984>, rot=<0.006 0.0168 -0.0329 -0.9993>}
 middle_01_l {parent=middle_metacarpal_l, loc=<7.3275 -0.2447 0.0129>, rot=<-0.0208 0.0042 0.0858 -0.9961>}
 middle_02_l {parent=middle_01_l, loc=<4.0123 0.1401 0.1175>, rot=<-0.0218 0.0258 0.3512 -0.9357>}
 middle_03_l {parent=middle_02_l, loc=<2.6077 0.0534 0.1373>, rot=<0.0082 0.0188 0.1708 -0.9851>}
 ring_metacarpal_l {parent=hand_l, loc=<2.2074 -0.3625 1.3638>, rot=<-0.0284 0.0722 -0.0458 -0.9959>}
 ring_01_l {parent=ring_metacarpal_l, loc=<7.0295 -0.6197 -0.1446>, rot=<-0.0489 -0.0252 0.1892 -0.9804>}
 ring_02_l {parent=ring_01_l, loc=<3.6174 0.4647 -0.1027>, rot=<0 0 0.2806 -0.9598>}
 ring_03_l {parent=ring_02_l, loc=<2.4152 0.114 0.0199>, rot=<0 0 0.1189 -0.9929>}
 pinky_metacarpal_l {parent=hand_l, loc=<1.7616 -0.9753 3.3814>, rot=<-0.0397 0.0814 -0.069 -0.9935>}
 pinky_01_l {parent=pinky_metacarpal_l, loc=<7.1709 -0.4831 -0.1094>, rot=<-0.0654 -0.0035 0.2242 -0.9723>}
 pinky_02_l {parent=pinky_01_l, loc=<2.8274 0.0893 -0.1043>, rot=<0.0037 0.0011 0.2854 -0.9584>}
 pinky_03_l {parent=pinky_02_l, loc=<2.2046 0.0307 -0.0078>, rot=<-0.0002 0.0098 0.1445 -0.9895>}
 lowerarm_twist_02_l {parent=lowerarm_l, loc=<10.6707 0 0>, rot=<0 0 0 1>}
 lowerarm_twist_01_l {parent=lowerarm_l, loc=<18.316 0 0>, rot=<0 0 0 1>}
 dyn_pomPom_l_A {parent=lowerarm_l, loc=<19.0086 0.3505 -0.4223>, rot=<0 0 0 1>}
 dyn_pomPom_l_B {parent=lowerarm_l, loc=<19.0086 0.3505 -0.4223>, rot=<0 0 0 1>}
 dyn_pomPom_l_C {parent=lowerarm_l, loc=<19.0086 0.3505 -0.4223>, rot=<0 0 0 1>}
 dyn_pomPom_l_D {parent=lowerarm_l, loc=<19.0086 0.3505 -0.4223>, rot=<0 0 0 1>}
 upperarm_twist_01_l {parent=upperarm_l, loc=<7.2135 -0.1134 0>, rot=<0 0 0 1>}
 deform_upperarm_bk_base_l {parent=upperarm_twist_01_l, loc=<-6.4053 -4.3746 0.314>, rot=<0 0 0 -1>}
 deform_upperarm_bk_l {parent=deform_upperarm_bk_base_l, loc=<0 0 0>, rot=<0 0 0 1>}
 deform_upperarm_fr_base_l {parent=upperarm_twist_01_l, loc=<-5.6312 5.092 0.3354>, rot=<0 0 0 -1>}
 deform_upperarm_fr_l {parent=deform_upperarm_fr_base_l, loc=<0 0 0>, rot=<0 0 0 1>}
 deform_upperarm_up_base_l {parent=upperarm_twist_01_l, loc=<-8.891 0.4825 3.9517>, rot=<0 0 0 -1>}
 deform_upperarm_up_l {parent=deform_upperarm_up_base_l, loc=<0 0 0>, rot=<0 0 0 1>}
 upperarm_twist_02_l {parent=upperarm_l, loc=<19.7998 -0.1134 0>, rot=<0 0 0 1>}
 clavicle_r {parent=spine_05, loc=<5.4831 -0.1053 3.2946>, rot=<0.7289 0.0465 -0.6829 0.009>}
 upperarm_r {parent=clavicle_r, loc=<-9.8658 1.2136 0.4105>, rot=<0.0164 -0.3245 -0.0269 -0.9454>}
 lowerarm_r {parent=upperarm_r, loc=<-26.465 0.8345 -0.259>, rot=<0.0036 0.0161 -0.2298 -0.9731>}
 hand_r {parent=lowerarm_r, loc=<-22.7355 0.3203 0.6106>, rot=<-0.6784 -0.0223 0.0735 -0.7307>}
 thumb_01_r {parent=hand_r, loc=<-1.3118 0.3191 2.459>, rot=<0.6546 -0.4042 0.0664 -0.6354>}
 thumb_02_r {parent=thumb_01_r, loc=<-5.9336 0.0448 -0.0401>, rot=<-0.0037 0.0213 0.298 -0.9543>}
 thumb_03_r {parent=thumb_02_r, loc=<-2.2223 -0.3682 0.0907>, rot=<0.0118 -0.013 0.0066 -0.9998>}
 index_metacarpal_r {parent=hand_r, loc=<-1.6535 -0.0183 2.96>, rot=<0.0158 -0.0124 -0.0173 -0.9996>}
 index_01_r {parent=index_metacarpal_r, loc=<-8.3008 -0.0357 0.0524>, rot=<-0.0044 0.0485 0.0924 -0.9945>}
 index_02_r {parent=index_01_r, loc=<-3.3587 0.3004 -0.0452>, rot=<-0.002 0.0027 0.3225 -0.9466>}
 index_03_r {parent=index_02_r, loc=<-2.5578 0.123 0.0252>, rot=<-0.002 0.0067 0.2558 -0.9667>}
 middle_metacarpal_r {parent=hand_r, loc=<-2.1856 0.1725 0.5984>, rot=<0.006 0.0168 -0.0329 -0.9993>}
 middle_01_r {parent=middle_metacarpal_r, loc=<-7.3275 0.2447 -0.0129>, rot=<-0.0208 0.0042 0.0858 -0.9961>}
 middle_02_r {parent=middle_01_r, loc=<-4.0123 -0.1401 -0.1175>, rot=<-0.0218 0.0258 0.3512 -0.9357>}
 middle_03_r {parent=middle_02_r, loc=<-2.6077 -0.0534 -0.1373>, rot=<0.0082 0.0188 0.1708 -0.9851>}
 ring_metacarpal_r {parent=hand_r, loc=<-2.2074 0.3625 -1.3638>, rot=<-0.0284 0.0722 -0.0458 -0.9959>}
 ring_01_r {parent=ring_metacarpal_r, loc=<-7.0295 0.6197 0.1446>, rot=<-0.0489 -0.0252 0.1892 -0.9804>}
 ring_02_r {parent=ring_01_r, loc=<-3.6174 -0.4647 0.1027>, rot=<0 0 0.2806 -0.9598>}
 ring_03_r {parent=ring_02_r, loc=<-2.4152 -0.114 -0.0199>, rot=<0 0 0.1189 -0.9929>}
 pinky_metacarpal_r {parent=hand_r, loc=<-1.7616 0.9753 -3.3814>, rot=<-0.0397 0.0814 -0.069 -0.9935>}
 pinky_01_r {parent=pinky_metacarpal_r, loc=<-7.1709 0.4831 0.1094>, rot=<-0.0654 -0.0035 0.2242 -0.9723>}
 pinky_02_r {parent=pinky_01_r, loc=<-2.8274 -0.0893 0.1043>, rot=<0.0037 0.0011 0.2854 -0.9584>}
 pinky_03_r {parent=pinky_02_r, loc=<-2.2046 -0.0307 0.0078>, rot=<-0.0002 0.0098 0.1445 -0.9895>}
 lowerarm_twist_02_r {parent=lowerarm_r, loc=<-10.6707 0 0>, rot=<0 0 0 1>}
 lowerarm_twist_01_r {parent=lowerarm_r, loc=<-18.316 0 0>, rot=<0 0 0 1>}
 dyn_pomPom_r_A {parent=lowerarm_r, loc=<-19.0086 -0.3505 0.4223>, rot=<0 -1 0 0>}
 dyn_pomPom_r_B {parent=lowerarm_r, loc=<-19.0086 -0.3505 0.4223>, rot=<0 -1 0 0>}
 dyn_pomPom_r_C {parent=lowerarm_r, loc=<-19.0086 -0.3505 0.4223>, rot=<0 -1 0 0>}
 dyn_pomPom_r_D {parent=lowerarm_r, loc=<-19.0086 -0.3505 0.4223>, rot=<0 -1 0 0>}
 upperarm_twist_01_r {parent=upperarm_r, loc=<-7.2135 0.1134 0>, rot=<0 0 0 1>}
 deform_upperarm_bk_base_r {parent=upperarm_twist_01_r, loc=<6.4053 4.3746 -0.314>, rot=<0 0 0 -1>}
 deform_upperarm_bk_r {parent=deform_upperarm_bk_base_r, loc=<0 0 0>, rot=<0 0 0 1>}
 deform_upperarm_fr_base_r {parent=upperarm_twist_01_r, loc=<5.6312 -5.092 -0.3354>, rot=<0 0 0 -1>}
 deform_upperarm_fr_r {parent=deform_upperarm_fr_base_r, loc=<0 0 0>, rot=<0 0 0 1>}
 deform_upperarm_up_base_r {parent=upperarm_twist_01_r, loc=<8.891 -0.4825 -3.9517>, rot=<0 0 0 -1>}
 deform_upperarm_up_r {parent=deform_upperarm_up_base_r, loc=<0 0 0>, rot=<0 0 0 1>}
 upperarm_twist_02_r {parent=upperarm_r, loc=<-19.7998 0.1134 0>, rot=<0 0 0 1>}
 attach_backpack {parent=spine_05, loc=<-0.1227 -13.4698 -0.0002>, rot=<-0.0093 -0.7071 0.0093 -0.707>}
 Backpack_BR {parent=attach_backpack, type=Null, loc=<0 2.7306 0>, rot=<0 0 0 1>}
 attach_cape {parent=spine_05, loc=<9.0162 -11.7114 -0.0013>, rot=<0.0616 -0.7044 -0.0616 -0.7044>}
 cape {parent=attach_cape, type=Null, loc=<0 0 0>, rot=<0 0 0 1>}
 thigh_l {parent=pelvis, loc=<-1.0855 -0.2109 -8.7053>, rot=<0.0009 0.0355 0.0155 -0.9993>}
 calf_l {parent=thigh_l, loc=<-42.5284 -0.0019 -0.7297>, rot=<0.0006 -0.0161 -0.0391 -0.9991>}
 foot_l {parent=calf_l, loc=<-46.7401 -2.5014 -0.592>, rot=<0.0427 -0.0238 0.0353 -0.9982>}
 ball_l {parent=foot_l, loc=<-3.5658 10.8066 -0.1544>, rot=<0.0001 -0.0001 -0.7063 -0.7079>}
 dyn_shoelace_bow_a_l {parent=foot_l, loc=<3.5295 8.6083 -0.8417>, rot=<0.1278 -0.7436 -0.1087 -0.6472>}
 dyn_shoelace_bow_b_l {parent=foot_l, loc=<3.4161 8.6374 1.3105>, rot=<0.6091 0.1433 0.7786 -0.0471>}
 dyn_shoelace_tail_a_l {parent=foot_l, loc=<2.5925 9.2512 -0.9945>, rot=<-0.1337 -0.8005 -0.3728 -0.4498>}
 dyn_shoelace_tail_b_l {parent=foot_l, loc=<2.5803 9.0573 1.1391>, rot=<0.4544 0.3839 0.7746 0.2149>}
 calf_twist_02_l {parent=calf_l, loc=<-19.1546 -1.7943 -0.2638>, rot=<0.0237 0.0095 -0.0035 -0.9997>}
 calf_twist_01_l {parent=calf_l, loc=<-35.8483 -1.9197 -0.5794>, rot=<0.0237 0.0095 -0.0035 -0.9997>}
 thigh_twist_01_l {parent=thigh_l, loc=<-27.952 0 0>, rot=<0.0013 0 0 -1>}
 thigh_r {parent=pelvis, loc=<-1.0855 -0.2109 8.7053>, rot=<0.0355 -0.0009 0.9993 0.0155>}
 calf_r {parent=thigh_r, loc=<42.5284 0.0019 0.7297>, rot=<0.0006 -0.0161 -0.0391 -0.9991>}
 foot_r {parent=calf_r, loc=<46.7401 2.5014 0.592>, rot=<0.0427 -0.0238 0.0353 -0.9982>}
 ball_r {parent=foot_r, loc=<3.5658 -10.8066 0.1544>, rot=<0.0001 -0.0001 -0.7063 -0.7079>}
 dyn_shoelace_bow_a_r {parent=foot_r, loc=<-3.3411 -8.6381 -1.3102>, rot=<0.7535 0.0909 -0.6387 0.1269>}
 dyn_shoelace_bow_b_r {parent=foot_r, loc=<-3.4122 -8.6316 0.8398>, rot=<-0.0856 0.6722 -0.0914 -0.7297>}
 dyn_shoelace_tail_a_r {parent=foot_r, loc=<-2.5803 -9.0573 -1.1391>, rot=<0.7723 -0.1996 -0.4608 0.389>}
 dyn_shoelace_tail_b_r {parent=foot_r, loc=<-2.5925 -9.2511 0.9941>, rot=<-0.3569 0.4797 0.1614 -0.7852>}
 calf_twist_02_r {parent=calf_r, loc=<19.1546 1.7943 0.2638>, rot=<0.0237 0.0095 -0.0035 -0.9997>}
 calf_twist_01_r {parent=calf_r, loc=<35.8483 1.9197 0.5794>, rot=<0.0237 0.0095 -0.0035 -0.9997>}
 thigh_twist_01_r {parent=thigh_r, loc=<27.952 0 0>, rot=<0.0013 0 0 -1>}
 dyn_skirt_fr_c_base {parent=pelvis, loc=<1.539 7.5425 0>, rot=<0.0127 0.707 0.707 0.0127>}
 dyn_skirt_fr_c_01 {parent=dyn_skirt_fr_c_base, loc=<0 0 0>, rot=<0 -0.0032 0 -1>}
 dyn_skirt_fr_c_02 {parent=dyn_skirt_fr_c_01, loc=<6.7902 0 0.617>, rot=<0 0.048 0 -0.9988>}
 dyn_skirt_fr_c_03 {parent=dyn_skirt_fr_c_02, loc=<5.7731 0 0.0672>, rot=<0 -0.0172 0 -0.9999>}
 dyn_skirt_fr_c_04 {parent=dyn_skirt_fr_c_03, loc=<5.2718 0 0.0273>, rot=<0 0 0 1>}
 dyn_skirt_fr_l_base {parent=pelvis, loc=<1.2496 5.5569 -9.8896>, rot=<-0.0086 -0.8528 -0.4914 -0.1766>}
 dyn_skirt_fr_l_01 {parent=dyn_skirt_fr_l_base, loc=<0 0 0>, rot=<0.0011 -0.0044 -0.0011 -1>}
 dyn_skirt_fr_l_02 {parent=dyn_skirt_fr_l_01, loc=<6.5572 0.2073 -0.4815>, rot=<-0.0179 -0.0624 -0.0308 -0.9974>}
 dyn_skirt_fr_l_03 {parent=dyn_skirt_fr_l_02, loc=<5.1744 0.0968 0.1788>, rot=<-0.0032 0.0204 -0.0083 -0.9998>}
 dyn_skirt_fr_l_04 {parent=dyn_skirt_fr_l_03, loc=<5.0858 0.0687 -0.0439>, rot=<0 0 0 1>}
 dyn_skirt_side_l_base {parent=pelvis, loc=<2.064 -1.6984 -13.8133>, rot=<-0.0176 -0.9798 0.0036 -0.1993>}
 dyn_skirt_side_l_01 {parent=dyn_skirt_side_l_base, loc=<0 0 0>, rot=<0 0.0009 0 -1>}
 dyn_skirt_side_l_02 {parent=dyn_skirt_side_l_01, loc=<6.3163 0 0.2496>, rot=<0 -0.0122 0 -0.9999>}
 dyn_skirt_side_l_03 {parent=dyn_skirt_side_l_02, loc=<4.7597 0 0.2136>, rot=<0 -0.0099 0 -1>}
 dyn_skirt_side_l_04 {parent=dyn_skirt_side_l_03, loc=<5.0546 0 0>, rot=<0 0 0 1>}
 dyn_skirt_bck_l_base {parent=pelvis, loc=<3.8404 -7.7735 -8.9662>, rot=<0.0489 -0.8703 0.4732 -0.1276>}
 dyn_skirt_bck_l_01 {parent=dyn_skirt_bck_l_base, loc=<0 0 0>, rot=<0.0007 0.0005 0.0018 -1>}
 dyn_skirt_bck_l_02 {parent=dyn_skirt_bck_l_01, loc=<7.7164 -0.4534 -0.1617>, rot=<0 -0.0311 -0.0083 -0.9995>}
 dyn_skirt_bck_l_03 {parent=dyn_skirt_bck_l_02, loc=<5.203 0.11 -0.0513>, rot=<-0.0795 -0.0251 -0.0089 -0.9965>}
 dyn_skirt_bck_l_04 {parent=dyn_skirt_bck_l_03, loc=<6.6779 -0.1629 0.0624>, rot=<0 0 0 1>}
 dyn_skirt_bck_c_base {parent=pelvis, loc=<3.5982 -9.7754 0>, rot=<0.0736 -0.7033 0.7033 -0.0736>}
 dyn_skirt_bck_c_01 {parent=dyn_skirt_bck_c_base, loc=<0 0 0>, rot=<0 0.0001 0 -1>}
 dyn_skirt_bck_c_02 {parent=dyn_skirt_bck_c_01, loc=<7.2974 0 0.2551>, rot=<0 -0.0138 0 -0.9999>}
 dyn_skirt_bck_c_03 {parent=dyn_skirt_bck_c_02, loc=<6.4404 0 -0.0995>, rot=<0 -0.0615 0 -0.9981>}
 dyn_skirt_bck_c_04 {parent=dyn_skirt_bck_c_03, loc=<6.814 0 -0.0457>, rot=<0 0 0 1>}
 dyn_skirt_bck_r_base {parent=pelvis, loc=<3.8407 -7.765 8.804>, rot=<0.1584 -0.5045 0.8485 -0.0219>}
 dyn_skirt_bck_r_01 {parent=dyn_skirt_bck_r_base, loc=<0 0 0>, rot=<0.0002 -0.0001 -0.0002 -1>}
 dyn_skirt_bck_r_02 {parent=dyn_skirt_bck_r_01, loc=<7.6992 0.8673 -0.2511>, rot=<-0.0574 -0.0465 -0.0269 -0.9969>}
 dyn_skirt_bck_r_03 {parent=dyn_skirt_bck_r_02, loc=<5.5694 0.0386 0.0708>, rot=<-0.0681 -0.023 0.0133 -0.9973>}
 dyn_skirt_bck_r_04 {parent=dyn_skirt_bck_r_03, loc=<6.323 0.0084 -0.0962>, rot=<0 0 0 1>}
 dyn_skirt_side_r_base {parent=pelvis, loc=<2.4926 -1.7137 13.6444>, rot=<0.2002 -0.0036 0.9796 0.0176>}
 dyn_skirt_side_r_01 {parent=dyn_skirt_side_r_base, loc=<0 0 0>, rot=<0 0 0 -1>}
 dyn_skirt_side_r_02 {parent=dyn_skirt_side_r_01, loc=<6.3192 0 0.2707>, rot=<0 -0.0121 0 -0.9999>}
 dyn_skirt_side_r_03 {parent=dyn_skirt_side_r_02, loc=<5.172 0 0.0742>, rot=<0 -0.01 0 -1>}
 dyn_skirt_side_r_04 {parent=dyn_skirt_side_r_03, loc=<5.0985 0 0.1173>, rot=<0 0 0 1>}
 dyn_skirt_fr_r_base {parent=pelvis, loc=<1.2496 5.5569 9.89>, rot=<0.1721 0.5436 0.8215 -0.0036>}
 dyn_skirt_fr_r_01 {parent=dyn_skirt_fr_r_base, loc=<0 0 0>, rot=<-0.0002 0.0005 0.0001 -1>}
 dyn_skirt_fr_r_02 {parent=dyn_skirt_fr_r_01, loc=<6.5571 -0.2644 -0.4525>, rot=<0.0354 -0.0588 0.0371 -0.997>}
 dyn_skirt_fr_r_03 {parent=dyn_skirt_fr_r_02, loc=<5.1744 -0.0809 0.1866>, rot=<-0.084 0.0215 0.0047 -0.9962>}
 dyn_skirt_fr_r_04 {parent=dyn_skirt_fr_r_03, loc=<5.0858 -0.0777 -0.0246>, rot=<0 0 0 1>}
]
variables [
 IsInFrontendEmote {type=double, default=0}
 IsPlayingFrontendIdle {type=double, default=0}
 L thighBack Alpha {type=double, default=0}
 L thighForward Alpha {type=double, default=0}
 L thighSide Alpha {type=double, default=0}
 L thighSideIn Alpha {type=double, default=0}
 L thighTwistIn Alpha {type=double, default=0}
 R thighBack Alpha {type=double, default=0}
 R thighForward Alpha {type=double, default=0}
 R thighSide Alpha {type=double, default=0}
 R thighSideIn Alpha {type=double, default=0}
 R thighTwistIn Alpha {type=double, default=0}
]
