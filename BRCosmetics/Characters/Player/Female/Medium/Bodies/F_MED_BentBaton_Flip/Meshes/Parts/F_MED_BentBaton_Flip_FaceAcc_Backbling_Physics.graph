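# Physics asset for the F_MED_BentBaton_Flip face accessory (Backbling variant by name), defining collision and joints for the character's dynamic fringe bones.
# Capsule bodies cover the head and each fringe bone.
# The left fringe is a three-bone chain from the head; the right root branches into three sub-chains.
# Twist is tightly limited throughout, while swing is stiffest at the right root, moderate on the left and loosest along the right branches.
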PhysicsAsset /BRCosmetics/Characters/Player/Female/Medium/Bodies/F_MED_BentBaton_Flip/Meshes/Parts/F_MED_BentBaton_Flip_FaceAcc_Backbling_Physics
bodies {
 dyn_fringe_l_01 [{shapeType=Capsule, length=1.5, radius=1.5, shapeName=dyn_fringe_l_01_capsule, center=<1.5391 0 0>, rotation=<rot 90 -160.5288 -160.5287>}]
 dyn_fringe_l_02 [{shapeType=Capsule, length=1.5, radius=1.5, shapeName=dyn_fringe_l_02_capsule, center=<1.5391 0 0>, rotation=<rot 90 179.9999 180>}]
 dyn_fringe_l_03 [{shapeType=Capsule, length=1.5, radius=1.5, shapeName=dyn_fringe_l_03_capsule, center=<1.455 0 0.4774>, rotation=<rot 80 179.9998 179.9999>}]
 dyn_fringe_r_01 [{shapeType=Capsule, length=1.5, radius=2.5, shapeName=dyn_fringe_r_01_capsule, center=<1.4748 0 1.1021>, rotation=<rot 85 179.9999 -180>}]
 dyn_fringe_r_02 [{shapeType=Capsule, length=1, radius=1.5, shapeName=dyn_fringe_r_02_capsule, center=<1.5374 0 0>, rotation=<rot 90 0 0>}]
 dyn_fringe_r_03 [{shapeType=Capsule, length=1.5, radius=1, shapeName=dyn_fringe_r_03_capsule, center=<2.3262 0.8141 -0.7624>, rotation=<rot 50 14.9998 0>}]
 dyn_fringe_r_04 [{shapeType=Capsule, length=1.5, radius=1, shapeName=dyn_fringe_r_04_capsule, center=<1.5374 0 0>, rotation=<rot 90 0 0>}]
 dyn_fringe_r_05 [{shapeType=Capsule, length=1, radius=1, shapeName=dyn_fringe_r_05_capsule, center=<1.729 0 1.3088>, rotation=<rot 45 179.9999 179.9999>}]
 dyn_fringe_r_06 [{shapeType=Capsule, length=2, radius=2, shapeName=dyn_fringe_r_06_capsule, center=<1.5374 -0.1388 -0.2178>, rotation=<rot 90 0 0>}]
 dyn_fringe_r_07 [{shapeType=Capsule, length=0.5, radius=2, shapeName=dyn_fringe_r_07_capsule, center=<1.5374 0.4335 0>, rotation=<rot 90 0 0>}]
 dyn_fringe_r_08 [{shapeType=Capsule, length=1, radius=1.5, shapeName=dyn_fringe_r_08_capsule, center=<1.7664 -0.7577 0>, rotation=<rot 90 180 -155>}]
 head [{shapeType=Capsule, length=7.5, radius=7, shapeName=head_capsule, center=<3.2991 2.492 0>, rotation=<rot 90 -6.7213 -6.7211>}]
}
constraints [
 {bone1Name=dyn_fringe_l_01, bone2Name=head, swing1LimitDegrees=20, swing2LimitDegrees=20, twistLimitDegrees=1}
 {bone1Name=dyn_fringe_l_02, bone2Name=dyn_fringe_l_01, swing1LimitDegrees=20, swing2LimitDegrees=20, twistLimitDegrees=1}
 {bone1Name=dyn_fringe_l_03, bone2Name=dyn_fringe_l_02, swing1LimitDegrees=20, swing2LimitDegrees=20, twistLimitDegrees=1}
 {bone1Name=dyn_fringe_r_01, bone2Name=head, swing1LimitDegrees=5, swing2LimitDegrees=5, twistLimitDegrees=1}
 {bone1Name=dyn_fringe_r_02, bone2Name=dyn_fringe_r_01, swing1LimitDegrees=45, swing2LimitDegrees=45, twistLimitDegrees=1}
 {bone1Name=dyn_fringe_r_03, bone2Name=dyn_fringe_r_02, swing1LimitDegrees=45, swing2LimitDegrees=45, twistLimitDegrees=1}
 {bone1Name=dyn_fringe_r_04, bone2Name=dyn_fringe_r_01, swing1LimitDegrees=45, swing2LimitDegrees=45, twistLimitDegrees=1}
 {bone1Name=dyn_fringe_r_05, bone2Name=dyn_fringe_r_04, swing1LimitDegrees=45, swing2LimitDegrees=45, twistLimitDegrees=1}
 {bone1Name=dyn_fringe_r_06, bone2Name=dyn_fringe_r_01, swing1LimitDegrees=45, swing2LimitDegrees=45, twistLimitDegrees=1}
 {bone1Name=dyn_fringe_r_07, bone2Name=dyn_fringe_r_06, swing1LimitDegrees=45, swing2LimitDegrees=45, twistLimitDegrees=1}
 {bone1Name=dyn_fringe_r_08, bone2Name=dyn_fringe_r_07, swing1LimitDegrees=45, swing2LimitDegrees=45, twistLimitDegrees=1}
]
