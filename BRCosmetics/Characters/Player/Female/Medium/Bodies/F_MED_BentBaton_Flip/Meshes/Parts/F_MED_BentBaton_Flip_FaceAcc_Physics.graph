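# Physics asset for the medium female BentBaton_Flip face accessory part, defining capsule bodies for ponytail and fringe hair physics.
# The ponytail chains six segments from a base with free angular motion on the head, mostly loosening down the chain.
# The left fringe is one chain, while the right splits into three branches from a tightly limited root.
# Unconstrained spine and upper-arm capsules, plus head capsules, provide body shapes for collision.
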PhysicsAsset /BRCosmetics/Characters/Player/Female/Medium/Bodies/F_MED_BentBaton_Flip/Meshes/Parts/F_MED_BentBaton_Flip_FaceAcc_Physics
bodies {
 dyn_fringe_l_01 [{shapeType=Capsule, length=1.5, radius=1.5, shapeName=dyn_fringe_l_01_capsule, center=<1.5391 0 0>, rotation=<rot 90 -160.5288 -160.5287>}]
 dyn_fringe_l_02 [{shapeType=Capsule, length=1.5, radius=1.5, shapeName=dyn_fringe_l_02_capsule, center=<1.5391 0 0>, rotation=<rot 90 179.9999 180>}]
 dyn_fringe_l_03 [{shapeType=Capsule, length=1.5, radius=1.5, shapeName=dyn_fringe_l_03_capsule, center=<1.455 0 0.4774>, rotation=<rot 80 179.9998 179.9999>}]
 dyn_fringe_r_01 [{shapeType=Capsule, length=1.5, radius=2.5, shapeName=dyn_fringe_r_01_capsule, center=<1.4748 0 1.1021>, rotation=<rot 85 179.9999 -180>}]
 dyn_fringe_r_02 [{shapeType=Capsule, length=1, radius=1.5, shapeName=dyn_fringe_r_02_capsule, center=<1.5374 0 0>, rotation=<rot 90 0 0>}]
 dyn_fringe_r_03 [{shapeType=Capsule, length=1.5, radius=1, shapeName=dyn_fringe_r_03_capsule, center=<2.3262 0.8141 -0.7624>, rotation=<rot 50 14.9998 0>}]
 dyn_fringe_r_04 [{shapeType=Capsule, length=1.5, radius=1, shapeName=dyn_fringe_r_04_capsule, center=<1.5374 0 0>, rotation=<rot 90 0 0>}]
 dyn_fringe_r_05 [{shapeType=Capsule, length=1, radius=1, shapeName=dyn_fringe_r_05_capsule, center=<1.729 0 1.3088>, rotation=<rot 45 179.9999 179.9999>}]
 dyn_fringe_r_06 [{shapeType=Capsule, length=2, radius=2, shapeName=dyn_fringe_r_06_capsule, center=<1.5374 -0.1388 -0.2178>, rotation=<rot 90 0 0>}]
 dyn_fringe_r_07 [{shapeType=Capsule, length=0.5, radius=2, shapeName=dyn_fringe_r_07_capsule, center=<1.5374 0.4335 0>, rotation=<rot 90 0 0>}]
 dyn_fringe_r_08 [{shapeType=Capsule, length=1, radius=1.5, shapeName=dyn_fringe_r_08_capsule, center=<1.7664 -0.7577 0>, rotation=<rot 90 180 -155>}]
 dyn_ponytail_base [{shapeType=Capsule, length=5, radius=5, shapeName=dyn_ponytail_base_capsule}]
 dyn_ponytail_center_01 [{shapeType=Capsule, length=3, radius=6, shapeName=dyn_ponytail_center_01_capsule, center=<3.0979 0 0>, rotation=<rot -90 0 0>}]
 dyn_ponytail_center_02 [{shapeType=Capsule, length=4, radius=6, shapeName=dyn_ponytail_center_02_capsule, center=<3.0979 0 0>, rotation=<rot -90 0 0>}]
 dyn_ponytail_center_03 [{shapeType=Capsule, length=3.5, radius=6.5, shapeName=dyn_ponytail_center_03_capsule, center=<3.0979 0 -0.3549>, rotation=<rot -90 0 0>}]
 dyn_ponytail_center_04 [{shapeType=Capsule, length=4, radius=6, shapeName=dyn_ponytail_center_04_capsule, center=<3.0979 0 0.5957>, rotation=<rot -90 180 180>}]
 dyn_ponytail_center_05 [{shapeType=Capsule, length=4.5, radius=5.5, shapeName=dyn_ponytail_center_05_capsule, center=<3.0979 0 1.1609>, rotation=<rot -90 0 0>}]
 dyn_ponytail_center_06 [{shapeType=Capsule, length=2.5, radius=5, shapeName=dyn_ponytail_center_06_capsule, center=<3.0979 0 1.5281>, rotation=<rot -90 0 0>}]
 head [
  {shapeType=Capsule, length=7.5, radius=7, shapeName=head_capsule, center=<3.2991 2.492 0>, rotation=<rot 90 -6.7213 -6.7211>}
  {shapeType=Capsule, length=35, radius=5, shapeName=head_capsule, center=<-5.1909 -3.4747 0>}
 ]
 spine_05 [{shapeType=Capsule, length=23.505, radius=9.005, shapeName=spine_05_capsule, center=<1.6915 0.5736 0>, rotation=<rot 0 -90 0>}]
 upperarm_l [{shapeType=Capsule, length=17, radius=5, shapeName=upperarm_l_capsule, center=<13.7718 0.2404 0>, rotation=<rot 90 -179.8508 179.1492>}]
 upperarm_r [{shapeType=Capsule, length=17, radius=5, shapeName=upperarm_r_capsule, center=<-13.7739 0 0>, rotation=<rot -90 0 0>}]
}
constraints [
 {
  bone1Name=dyn_ponytail_base
  bone2Name=head
  swing1LimitDegrees=45
  swing1Motion=Free
  swing2LimitDegrees=45
  swing2Motion=Free
  twistLimitDegrees=45
  twistMotion=Free
 }
 {bone1Name=dyn_ponytail_center_01, bone2Name=dyn_ponytail_base, swing1LimitDegrees=1, swing2LimitDegrees=25, twistLimitDegrees=1}
 {bone1Name=dyn_ponytail_center_02, bone2Name=dyn_ponytail_center_01, swing1LimitDegrees=20, swing2LimitDegrees=20, twistLimitDegrees=1}
 {bone1Name=dyn_ponytail_center_03, bone2Name=dyn_ponytail_center_02, swing1LimitDegrees=35, swing2LimitDegrees=45, twistLimitDegrees=1}
 {bone1Name=dyn_ponytail_center_04, bone2Name=dyn_ponytail_center_03, swing1LimitDegrees=35, swing2LimitDegrees=45, twistLimitDegrees=1}
 {bone1Name=dyn_ponytail_center_05, bone2Name=dyn_ponytail_center_04, swing1LimitDegrees=35, swing2LimitDegrees=45, twistLimitDegrees=1}
 {bone1Name=dyn_ponytail_center_06, bone2Name=dyn_ponytail_center_05, swing1LimitDegrees=35, swing2LimitDegrees=45, twistLimitDegrees=1}
 {bone1Name=dyn_fringe_l_01, bone2Name=head, swing1LimitDegrees=20, swing2LimitDegrees=20, twistLimitDegrees=1}
 {bone1Name=dyn_fringe_l_02, bone2Name=dyn_fringe_l_01, swing1LimitDegrees=20, swing2LimitDegrees=20, twistLimitDegrees=1}
 {bone1Name=dyn_fringe_l_03, bone2Name=dyn_fringe_l_02, swing1LimitDegrees=20, swing2LimitDegrees=20, twistLimitDegrees=1}
 {bone1Name=dyn_fringe_r_01, bone2Name=head, swing1LimitDegrees=5, swing2LimitDegrees=5, twistLimitDegrees=1}
 {bone1Name=dyn_fringe_r_02, bone2Name=dyn_fringe_r_01, swing1LimitDegrees=45, swing2LimitDegrees=45, twistLimitDegrees=1}
 {bone1Name=dyn_fringe_r_03, bone2Name=dyn_fringe_r_02, swing1LimitDegrees=45, swing2LimitDegrees=45, twistLimitDegrees=1}
 {bone1Name=dyn_fringe_r_04, bone2Name=dyn_fringe_r_01, swing1LimitDegrees=45, swing2LimitDegrees=45, twistLimitDegrees=1}
 {bone1Name=dyn_fringe_r_05, bone2Name=dyn_fringe_r_04, swing1LimitDegrees=45, swing2LimitDegrees=45, twistLimitDegrees=1}
 {bone1Name=dyn_fringe_r_06, bone2Name=dyn_fringe_r_01, swing1LimitDegrees=45, swing2LimitDegrees=45, twistLimitDegrees=1}
 {bone1Name=dyn_fringe_r_07, bone2Name=dyn_fringe_r_06, swing1LimitDegrees=45, swing2LimitDegrees=45, twistLimitDegrees=1}
 {bone1Name=dyn_fringe_r_08, bone2Name=dyn_fringe_r_07, swing1LimitDegrees=45, swing2LimitDegrees=45, twistLimitDegrees=1}
]
