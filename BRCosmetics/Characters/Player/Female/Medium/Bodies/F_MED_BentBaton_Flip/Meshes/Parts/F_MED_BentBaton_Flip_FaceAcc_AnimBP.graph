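# Post-physics AnimBP for the face accessory mesh of the BentBaton Flip female cosmetic, adding secondary motion to ponytail and bows.
# Copied parent and body poses feed two RigidBody nodes (standard and backbling physics assets), then per-bone AnimDynamics springs.
# The cached result gets additive bone offsets and weighted spine, neck and head constraints along the center ponytail.
# Events retune simulation for skydiving and front end, and set alpha variables from a backbling check.
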AnimBlueprint /BRCosmetics/Characters/Player/Female/Medium/Bodies/F_MED_BentBaton_Flip/Meshes/Parts/F_MED_BentBaton_Flip_FaceAcc_AnimBP
target_skeleton=&/F_MED_BentBaton_Flip_FaceAcc_Skeleton
graphs [
 AnimGraph {nodes [
   n1 = "Transform (Modify) Bone - Bone: dyn_ponytail_center_01"(
    ComponentPose=Local To Component(LocalPose=Use cached pose 'DynamicsCache'())
    Alpha=1
    Translation="0,0,0"
    Rotation="0,0.001,0"
    Scale="1,1,1"
   ) {
    rotationMode=BMM_Additive
    rotationSpace=BCS_BoneSpace
    scaleMode=BMM_Replace
    scaleSpace=BCS_BoneSpace
    translationMode=BMM_Additive
    translationSpace=BCS_BoneSpace
    alphaBoolBlend {blendOption=Linear}
    boneToModify {boneName=dyn_ponytail_center_01}
    rotation {yaw=0.001}
   }
   n2 = "Transform (Modify) Bone - Bone: dyn_ponytail_center_02"(ComponentPose=@n1, Alpha=1, Translation="0,0,0", Rotation="0,0,0", Scale="1,1,1") {
    rotationMode=BMM_Additive
    rotationSpace=BCS_BoneSpace
    scaleMode=BMM_Replace
    scaleSpace=BCS_BoneSpace
    translationMode=BMM_Additive
    translationSpace=BCS_BoneSpace
    alphaBoolBlend {blendOption=Linear}
    boneToModify {boneName=dyn_ponytail_center_02}
   }
   n3 = "Transform (Modify) Bone - Bone: dyn_ponytail_center_03"(ComponentPose=@n2, Alpha=1, Translation="-1,-1.163,-1.796", Rotation="0,0,0", Scale="1,1,1") {
    rotationMode=BMM_Additive
    rotationSpace=BCS_BoneSpace
    scaleMode=BMM_Replace
    scaleSpace=BCS_BoneSpace
    translationMode=BMM_Additive
    translationSpace=BCS_BoneSpace
    alphaBoolBlend {blendOption=Linear}
    boneToModify {boneName=dyn_ponytail_center_03}
    translation=<-1 -1.163 -1.796>
   }
   n4 = "Constraint - dyn_ponytail_center_04 (PP)"!(ComponentPose=@n3, Alpha=1, ConstraintWeights_0=0.5, ConstraintWeights_1:float=0.5) -> (Pose) {
    alphaBoolBlend {blendOption=Linear}
    boneToModify {boneName=dyn_ponytail_center_04}
    constraintSetup [{targetBone {boneName=spine_05}}, {targetBone {boneName=head}}]
    constraintWeights [1,1]
   }
   n5 = "Constraint - dyn_ponytail_center_05 (PP)"!(ComponentPose=@n4, Alpha=1, ConstraintWeights_0=0.8, ConstraintWeights_1:float=0.2) -> (Pose) {
    alphaBoolBlend {blendOption=Linear}
    boneToModify {boneName=dyn_ponytail_center_05}
    constraintSetup [{targetBone {boneName=spine_05}}, {targetBone {boneName=head}}]
    constraintWeights [1,1]
   }
   n6 = "Constraint - dyn_ponytail_center_06 (PP)"!(ComponentPose=@n5, Alpha=1, ConstraintWeights_0=0.9, ConstraintWeights_1:float=0.1) -> (Pose) {
    alphaBoolBlend {blendOption=Linear}
    boneToModify {boneName=dyn_ponytail_center_06}
    constraintSetup [{targetBone {boneName=spine_05}}, {targetBone {boneName=neck_01}}]
    constraintWeights [1,1]
   }
   n7 = "Transform (Modify) Bone - Bone: dyn_ponytail_center_04"(ComponentPose=@n6, Alpha=1, Translation="-1,0,-2", Rotation="5.001,-0.280,0.015", Scale="1,1,1") {
    rotationMode=BMM_Additive
    rotationSpace=BCS_BoneSpace
    scaleMode=BMM_Replace
    scaleSpace=BCS_BoneSpace
    translationMode=BMM_Additive
    translationSpace=BCS_BoneSpace
    alphaBoolBlend {blendOption=Linear}
    boneToModify {boneName=dyn_ponytail_center_04}
    rotation=<rot 5.001 -0.28 0.015>
    translation {x=-1, z=-2}
   }
   n8 = "Transform (Modify) Bone - Bone: dyn_ponytail_center_05"(ComponentPose=@n7, Alpha=1, Translation="-1.619,0,-1", Rotation="0,0,0", Scale="1,1,1") {
    rotationMode=BMM_Additive
    rotationSpace=BCS_BoneSpace
    scaleMode=BMM_Replace
    scaleSpace=BCS_BoneSpace
    translationMode=BMM_Additive
    translationSpace=BCS_BoneSpace
    alphaBoolBlend {blendOption=Linear}
    boneToModify {boneName=dyn_ponytail_center_05}
    translation {x=-1.619, z=-1}
   }
   n9 = "Transform (Modify) Bone - Bone: dyn_ponytail_center_06"(ComponentPose=@n8, Alpha=1, Translation="0,0,0", Rotation="0,0,0.014", Scale="1,1,1") {
    rotationMode=BMM_Additive
    rotationSpace=BCS_BoneSpace
    scaleMode=BMM_Replace
    scaleSpace=BCS_BoneSpace
    translationMode=BMM_Additive
    translationSpace=BCS_BoneSpace
    alphaBoolBlend {blendOption=Linear}
    boneToModify {boneName=dyn_ponytail_center_06}
    rotation {roll=0.014}
   }
   Output Pose(Result=Component To Local(ComponentPose=@n9)) {layerGroup=None}
   n10 = Copy Pose From Mesh(SourceMeshComponent=Get Body Mesh() -> (Body Mesh:Skeletal Mesh Component Object Reference)) {bUseAttachedParent=true}
   n11 = Copy Pose From Mesh() {bCopyCurves=true, bUseAttachedParent=true}
   n12 = RigidBody!(
    ComponentPose=Local To Component(LocalPose=Layered blend per bone(BasePose=@n11, BlendPoses_0=@n10) {layerSetup [{branchFilters [{boneName=upperarm_l}, {boneName=upperarm_r}]}]})
    Alpha
    OverrideWorldGravity:Vector=Get Override World Gravity() -> (Override World Gravity:Vector)
    ExternalForce
    SimSpaceSettings:Sim Space Settings Structure=(WorldAlpha=0.35,VelocityScaleZ=1,DampingAlpha=1,MaxLinearVelocity=10000,MaxAngularVelocity=10000,MaxLinearAcceleration=10000,MaxAngularAcceleration=10000,ExternalLinearDragV=(X=0,Y=0,Z=0),ExternalLinearVelocity=(X=0,Y=0,Z=0),ExternalAngularVelocity=(X=0,Y=0,Z=0))
   ) -> (Pose) {
    bClampLinearTranslationLimitToRefPose=true
    bOverrideWorldGravity=true
    overridePhysicsAsset=&/F_MED_BentBaton_Flip_FaceAcc_Physics
    simulationSpace=BaseBoneSpace
    alphaBoolBlend {blendOption=Linear}
    baseBoneRef {boneName=neck_02}
    componentAppliedLinearAccClamp=<3000 3000 3000>
    overrideWorldGravity {z=-980}
    simSpaceSettings {worldAlpha=0.35}
   }
   n13 = RigidBody(
    ComponentPose=@n12
    OverrideWorldGravity=Get Override World Gravity()
    SimSpaceSettings=(WorldAlpha=0.35,VelocityScaleZ=1,DampingAlpha=1,MaxLinearVelocity=10000,MaxAngularVelocity=10000,MaxLinearAcceleration=10000,MaxAngularAcceleration=10000,ExternalLinearDragV=(X=0,Y=0,Z=0),ExternalLinearVelocity=(X=0,Y=0,Z=0),ExternalAngularVelocity=(X=0,Y=0,Z=0))
   ) {
    bClampLinearTranslationLimitToRefPose=true
    bOverrideWorldGravity=true
    overridePhysicsAsset=&/F_MED_BentBaton_Flip_FaceAcc_Backbling_Physics
    simulationSpace=BaseBoneSpace
    alphaBoolBlend {blendOption=Linear}
    baseBoneRef {boneName=neck_02}
    componentAppliedLinearAccClamp=<3000 3000 3000>
    overrideWorldGravity {z=-980}
    simSpaceSettings {worldAlpha=0.35}
   }
   n14 = Anim Dynamics - Bone: dyn_bow_l_01!(SimulationSpace:AnimPhysSimSpaceType Enum=Component, ComponentPose=@n13, Alpha=1, ExternalForce="0,0,0") -> (Pose) {
    angularDampingOverride=1
    angularSpringConstant=5000
    bAngularSpring=true
    bGravityOverrideInSimSpace=true
    bOverrideAngularDamping=true
    bUseGravityOverride=true
    gravityScale=0.35
    numSolverIterationsPreUpdate=3
    alphaBoolBlend {blendOption=Linear}
    boundBone {boneName=dyn_bow_l_01}
    componentLinearAccScale=<1 1 1>
    componentLinearVelScale=<1 1 1>
    gravityOverride {z=-980}
    physicsBodyDefinitions [{
      boundBone {boneName=dyn_bow_l_01}
      boxExtents=<25 25 25>
      constraintSetup {angularLimitsMax=<0 5 0>, angularLimitsMin=<0 -45 0>, angularTarget=<1 0 0>}
      localJointOffset=<1 0 0>
     }]
   }
   n15 = Anim Dynamics - Bone: dyn_bow_l_02!(SimulationSpace:AnimPhysSimSpaceType Enum, ComponentPose=@n14, Alpha=1, ExternalForce="0,0,0") -> (Pose) {
    angularDampingOverride=1
    angularSpringConstant=5000
    bAngularSpring=true
    bGravityOverrideInSimSpace=true
    bOverrideAngularDamping=true
    bUseGravityOverride=true
    gravityScale=0.35
    numSolverIterationsPreUpdate=3
    simulationSpace=RootRelative
    alphaBoolBlend {blendOption=Linear}
    boundBone {boneName=dyn_bow_l_02}
    componentLinearAccScale=<1 1 1>
    componentLinearVelScale=<1 1 1>
    gravityOverride {z=-980}
    physicsBodyDefinitions [{
      boundBone {boneName=dyn_bow_l_02}
      boxExtents=<25 25 25>
      constraintSetup {angularLimitsMax=<0 5 0>, angularLimitsMin=<0 -45 0>, angularTarget=<1 0 0>}
      localJointOffset=<1 0 0>
     }]
   }
   n16 = Anim Dynamics - Bone: dyn_bow_r_01!(SimulationSpace:AnimPhysSimSpaceType Enum, ComponentPose=@n15, Alpha=1, ExternalForce="0,0,0") -> (Pose) {
    angularDampingOverride=1
    angularSpringConstant=5000
    bAngularSpring=true
    bGravityOverrideInSimSpace=true
    bOverrideAngularDamping=true
    bUseGravityOverride=true
    gravityScale=0.35
    numSolverIterationsPreUpdate=3
    simulationSpace=RootRelative
    alphaBoolBlend {blendOption=Linear}
    boundBone {boneName=dyn_bow_r_01}
    componentLinearAccScale=<1 1 1>
    componentLinearVelScale=<1 1 1>
    gravityOverride {z=-980}
    physicsBodyDefinitions [{
      boundBone {boneName=dyn_bow_r_01}
      boxExtents=<20 20 20>
      constraintSetup {angularLimitsMax=<0 5 0>, angularLimitsMin=<0 -45 0>, angularTarget=<1 0 0>}
      localJointOffset=<1 0 0>
     }]
   }
   n17 = Anim Dynamics - Bone: dyn_bow_r_02!(SimulationSpace:AnimPhysSimSpaceType Enum, ComponentPose=@n16, Alpha=1, ExternalForce="0,0,0") -> (Pose) {
    angularDampingOverride=1
    angularSpringConstant=5000
    bAngularSpring=true
    bGravityOverrideInSimSpace=true
    bOverrideAngularDamping=true
    bUseGravityOverride=true
    gravityScale=0.35
    numSolverIterationsPreUpdate=3
    simulationSpace=RootRelative
    alphaBoolBlend {blendOption=Linear}
    boundBone {boneName=dyn_bow_r_02}
    componentLinearAccScale=<1 1 1>
    componentLinearVelScale=<1 1 1>
    gravityOverride {z=-980}
    physicsBodyDefinitions [{
      boundBone {boneName=dyn_bow_r_02}
      boxExtents=<20 20 20>
      constraintSetup {angularLimitsMax=<0 5 0>, angularLimitsMin=<0 -45 0>, angularTarget=<1 0 0>}
      localJointOffset=<1 0 0>
     }]
   }
   n18 = Anim Dynamics - Bone: dyn_ponytail_piece_A!(SimulationSpace:AnimPhysSimSpaceType Enum, ComponentPose=@n17, Alpha=1, ExternalForce="0,0,0") -> (Pose) {
    angularDampingOverride=1
    angularSpringConstant=5000
    bAngularSpring=true
    bGravityOverrideInSimSpace=true
    bOverrideAngularDamping=true
    bUseGravityOverride=true
    gravityScale=0.35
    numSolverIterationsPreUpdate=3
    simulationSpace=RootRelative
    alphaBoolBlend {blendOption=Linear}
    boundBone {boneName=dyn_ponytail_piece_A}
    componentLinearAccScale=<1 1 1>
    componentLinearVelScale=<1 1 1>
    gravityOverride {z=-980}
    physicsBodyDefinitions [{
      boundBone {boneName=dyn_ponytail_piece_A}
      boxExtents=<25 25 25>
      constraintSetup {angularLimitsMax=<0 5 0>, angularLimitsMin=<0 -45 0>, angularTarget=<1 0 0>}
      localJointOffset=<1 0 0>
     }]
   }
   n19 = Anim Dynamics - Bone: dyn_ponytail_piece_B!(SimulationSpace:AnimPhysSimSpaceType Enum, ComponentPose=@n18, Alpha=1, ExternalForce="0,0,0") -> (Pose) {
    angularDampingOverride=1
    angularSpringConstant=5000
    bAngularSpring=true
    bGravityOverrideInSimSpace=true
    bOverrideAngularDamping=true
    bUseGravityOverride=true
    gravityScale=0.35
    numSolverIterationsPreUpdate=3
    simulationSpace=RootRelative
    alphaBoolBlend {blendOption=Linear}
    boundBone {boneName=dyn_ponytail_piece_B}
    componentLinearAccScale=<1 1 1>
    componentLinearVelScale=<1 1 1>
    gravityOverride {z=-980}
    physicsBodyDefinitions [{
      boundBone {boneName=dyn_ponytail_piece_B}
      boxExtents=<25 25 25>
      constraintSetup {angularLimitsMax=<0 5 0>, angularLimitsMin=<0 -45 0>, angularTarget=<1 0 0>}
      localJointOffset=<1 0 0>
     }]
   }
   n20 = Anim Dynamics - Bone: dyn_ponytail_piece_C!(SimulationSpace:AnimPhysSimSpaceType Enum, ComponentPose=@n19, Alpha=1, ExternalForce="0,0,0") -> (Pose) {
    angularDampingOverride=1
    angularSpringConstant=5000
    bAngularSpring=true
    bGravityOverrideInSimSpace=true
    bOverrideAngularDamping=true
    bUseGravityOverride=true
    gravityScale=0.35
    numSolverIterationsPreUpdate=3
    simulationSpace=RootRelative
    alphaBoolBlend {blendOption=Linear}
    boundBone {boneName=dyn_ponytail_piece_C}
    componentLinearAccScale=<1 1 1>
    componentLinearVelScale=<1 1 1>
    gravityOverride {z=-980}
    physicsBodyDefinitions [{
      boundBone {boneName=dyn_ponytail_piece_C}
      boxExtents=<20 20 20>
      constraintSetup {angularLimitsMax=<0 5 0>, angularLimitsMin=<0 -45 0>, angularTarget=<1 0 0>}
      localJointOffset=<1 0 0>
     }]
   }
   n21 = Anim Dynamics - Bone: dyn_ponytail_piece_D!(SimulationSpace:AnimPhysSimSpaceType Enum, ComponentPose=@n20, Alpha=1, ExternalForce="0,0,0") -> (Pose) {
    angularDampingOverride=1
    angularSpringConstant=5000
    bAngularSpring=true
    bGravityOverrideInSimSpace=true
    bOverrideAngularDamping=true
    bUseGravityOverride=true
    gravityScale=0.35
    numSolverIterationsPreUpdate=3
    simulationSpace=RootRelative
    alphaBoolBlend {blendOption=Linear}
    boundBone {boneName=dyn_ponytail_piece_D}
    componentLinearAccScale=<1 1 1>
    componentLinearVelScale=<1 1 1>
    gravityOverride {z=-980}
    physicsBodyDefinitions [{
      boundBone {boneName=dyn_ponytail_piece_D}
      boxExtents=<20 20 20>
      constraintSetup {angularLimitsMax=<0 5 0>, angularLimitsMin=<0 -45 0>, angularTarget=<1 0 0>}
      localJointOffset=<1 0 0>
     }]
   }
   n22 = Anim Dynamics - Bone: dyn_ponytail_piece_E!(SimulationSpace:AnimPhysSimSpaceType Enum, ComponentPose=@n21, Alpha=1, ExternalForce="0,0,0") -> (Pose) {
    angularDampingOverride=1
    angularSpringConstant=5000
    bAngularSpring=true
    bGravityOverrideInSimSpace=true
    bOverrideAngularDamping=true
    bUseGravityOverride=true
    gravityScale=0.35
    numSolverIterationsPreUpdate=3
    simulationSpace=RootRelative
    alphaBoolBlend {blendOption=Linear}
    boundBone {boneName=dyn_ponytail_piece_E}
    componentLinearAccScale=<1 1 1>
    componentLinearVelScale=<1 1 1>
    gravityOverride {z=-980}
    physicsBodyDefinitions [{
      boundBone {boneName=dyn_ponytail_piece_E}
      boxExtents=<20 20 20>
      constraintSetup {angularLimitsMax=<0 5 0>, angularLimitsMin=<0 -45 0>, angularTarget=<1 0 0>}
      localJointOffset=<1 0 0>
     }]
   }
   Save cached pose 'DynamicsCache'(Pose=Component To Local(ComponentPose=@n22)) {cacheName=DynamicsCache}
  ]}
 EventGraph {nodes [
   Event OnBeginSkydiving()
   > Set Override World Gravity(Override World Gravity:Vector="0,0,-200") -> (Output_Get:Vector)
   > Set Sim Space Settings(Sim Space Settings:Sim Space Settings Structure=Make Sim Space Settings!(
     WorldAlpha:float=0.2
     VelocityScaleZ:float=1
     DampingAlpha:float=1
     MaxLinearVelocity:float=800
     MaxAngularVelocity:float=8
     MaxLinearAcceleration:float=10000
     MaxAngularAcceleration:float=10000
     ExternalLinearDragV:Vector="0,0,0"
     ExternalLinearVelocity:Vector="0,0,0"
     ExternalAngularVelocity:Vector="0,0,0"
    ) -> (SimSpaceSettings:Sim Space Settings Structure) {is=title K2Node_MakeStruct}) -> (Output_Get:Sim Space Settings Structure)
   Event OnEndSkydiving()
   > Set Override World Gravity(Override World Gravity="0,0,-980")
   > Set Sim Space Settings(Sim Space Settings=Make Sim Space Settings(
     WorldAlpha=0.35
     VelocityScaleZ=1
     DampingAlpha=1
     MaxLinearVelocity=800
     MaxAngularVelocity=8
     MaxLinearAcceleration=10000
     MaxAngularAcceleration=10000
     ExternalLinearDragV="0,0,0"
     ExternalLinearVelocity="0,0,0"
     ExternalAngularVelocity="0,0,0"
    ))
   Event OnCharacterCustomizationLoaded()
   > n1 = Sequence()
   > n2 = Branch(Condition=Get bIsInFrontEnd())
   > Set Sim Space Settings(Sim Space Settings=Make Sim Space Settings(
     WorldAlpha=0.5
     VelocityScaleZ=1
     DampingAlpha=1
     MaxLinearVelocity=800
     MaxAngularVelocity=8
     MaxLinearAcceleration=5000
     MaxAngularAcceleration=5000
     ExternalLinearDragV="5,5,5"
     ExternalLinearVelocity="0,0,0"
     ExternalAngularVelocity="0,0,0"
    ))
   > Set Simulation Space(Simulation Space:AnimPhysSimSpaceType Enum=World) -> (Output_Get:AnimPhysSimSpaceType Enum)
   Set Sim Space Settings(
    execute=@n2.else
    Sim Space Settings=Make Sim Space Settings(
     WorldAlpha=0.35
     VelocityScaleZ=1
     DampingAlpha=1
     MaxLinearVelocity=800
     MaxAngularVelocity=8
     MaxLinearAcceleration=10000
     MaxAngularAcceleration=10000
     ExternalLinearDragV="0,0,0"
     ExternalLinearVelocity="0,0,0"
     ExternalAngularVelocity="0,0,0"
    )
   )
   > Set Simulation Space()
   n3 = Branch(execute=@n1.then_1, Condition=CheckForBackbling())
   > Set backbling alpha(backbling alpha:double=1) -> (Output_Get:double)
   > Set rban alpha(rban alpha:double=0) -> (Output_Get:double)
   > Set backbling rban alpha(backbling rban alpha:double=1) -> (Output_Get:double)
   Set backbling alpha(execute=@n3.else, backbling alpha=0)
   > Set rban alpha(rban alpha=1)
   > Set backbling rban alpha(backbling rban alpha=0)
   Event BlueprintUpdateAnimation()
  ]}
]
variables [
 backbling alpha {type=number}
 backbling rban alpha {type=number}
 Body Mesh {type=SkeletalMeshComponent}
 Override World Gravity {type=Vector, default=<0 0 -980>}
 rban alpha {type=number}
 Sim Space Settings {
  type=SimSpaceSettings
  default {
   dampingAlpha=1
   maxAngularAcceleration=10000
   maxAngularVelocity=8
   maxLinearAcceleration=10000
   maxLinearVelocity=800
   velocityScaleZ=1
   worldAlpha=0.35
   externalLinearDragV=<0 0 0>
  }
 }
 Simulation Space {type=AnimPhysSimSpaceType, default=RootRelative}
]
